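# Additive unlit light flare material for Labs Danger Room, drawing a tinted star texture as a camera-facing sprite.
# World position offset combines the Sprite function with an instance-space offset and a camera offset.
# Camera distance drives a fade that shrinks and dims the flare up close.
# Brightness also follows the FlareOn collection parameter, depth fade and view-to-normal angle, while color runs through HDR remapping and emissive adjustment functions.
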Material /Hermes/Sets/Labs/Meshes/Lightflare/M_Labs_DangerRoom_Light_Flare
graph {
 properties {
  assetImportData="<self>:InterchangeAssetImportData_0"
  bUsedWithInstancedStaticMeshes=true
  bUsedWithNanite=true
  bUsedWithNiagaraSprites=true
  blendMode=BLEND_Additive
  shadingModel=MSM_Unlit
  usedShadingModels=Unlit
 }
 nodes [
  n1 = LocalPosition()
  n2 = TextureCoordinate()
  n3 = Length(None=Subtract(A=CameraPositionWS(), B=ObjectPositionWS()) {constB=100}) {is=class MaterialExpressionLength}
  n4 = ScalarParameter() {defaultValue=50, parameterName=Fade Offset}
  n5 = Saturate(None=Divide(A=Subtract(A=@n3, B=@n4), B=ScalarParameter() {defaultValue=250, parameterName=Fade Length}))
  n6 = LinearInterpolate(
   A=Constant2Vector()
   B=ComponentMask(None=VectorParameter() {parameterName=Sprite Size, defaultValue=<rgba 2 2 0 1>}.RGB) {g=true, r=true}
   Alpha=@n5
  )
  n7 = StaticBool() {value=true}
  n8 = VectorParameter() {parameterName=XYZ, defaultValue=<rgba 0 5 0 1>}
  Add(
   A=Add(
    A=MaterialFunctionCall!(Center Location=@n1.XYZ, 0-1 UVs=@n2, XY scale=@n6, OverridePivot=@n7, "Normals (Optional)") {materialFunction=&Sprite}.World position offset
    B=Transform(None=@n8.RGB) {transformSourceType=TRANSFORMSOURCE_Instance}
   )
   B=MaterialFunctionCall(Offset Amount=ScalarParameter() {parameterName=CameraOffset}) {materialFunction=/Engine/Functions/Engine_MaterialFunctions02/WorldPositionOffset/CameraOffset}.Camera Offset
  )
  n9 = LinearInterpolate(Alpha=CollectionParameter() {collection=&/MPC_DangerRoom, parameterName=FlareOn}) {constB=50}
  n10 = DepthFade(FadeDistance=ScalarParameter() {defaultValue=100, parameterName=Depth Fade})
  n11 = CameraVectorWS()
  n12 = VertexNormalWS()
  LinearInterpolate(B=Multiply(A=Multiply(A=@n9, B=@n10), B=DotProduct(A=@n11, B=@n12)), Alpha=@n5)
  n13 = Multiply(
   A=VectorParameter() {parameterName=Color, defaultValue=<rgba 0 0.2085 1 1>}.RGB
   B=TextureSampleParameter2D(UVs=TextureCoordinate()) {group=BaseFlare, parameterName=BaseFlare, texture=/Game/Effects/Fort_Effects/Textures/Flares/T_GlowStar_5pt}.RGB
  )
  MaterialFunctionCall(Emissive=MaterialFunctionCall(LDR Input=Clamp(None=@n13), HDR Tint=@n13) {materialFunction=&DeriveHDRfromLDR}.RemappedRange) {materialFunction=&MF_FX_EmissiveAdjustments}
  ObjectLocalBounds()
 ]
}
